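# Material function for charm accessories that outputs an emissive color from the CharmEmissiveTexture parameter, tinted by EmissiveTextureModulate.
# The UseCroppedEmissive static switch chooses between a plain texture sample and the masked output of the engine's TextureCropping function.
# One vector parameter sets the crop rectangle, with RG as the upper-left corner.
# The lower-right corner uses B for both axes, despite the parameter name indicating BA.
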MaterialFunction /Game/Accessories/Charms/MaterialFunctions/MF_Charm_Emissive
graph {nodes [
  n1 = VectorParameter() {group=EmissiveTexture, parameterName=EmissiveUVs_RG_UpperLeftCorner_BA_LowerRightCorner, defaultValue=<rgba 0 0 1 1>}
  n2 = MaterialFunctionCall!(
   UVs
   TextureIn=TextureObjectParameter() {group=EmissiveTexture, parameterName=CharmEmissiveTexture, texture=&T_BlacksRGB}
   UpperLeftCorner=ComponentMask(None=@n1.RGB) {g=true, r=true}
   LowerRightCorner=AppendVector(A=@n1.B, B=@n1.B)
   ExportFloat4
  ) {materialFunction=/Engine/Functions/Engine_MaterialFunctions01/Texturing/TextureCropping}
  n3 = TextureSampleParameter2D() {parameterName=CharmEmissiveTexture, samplerSource=SSM_Wrap_WorldGroupSettings, texture=&T_BlacksRGB}
  FunctionOutput(None=Multiply(
    A=VectorParameter() {group=EmissiveTexture, parameterName=EmissiveTextureModulate, defaultValue=<rgba 1 1 1 1>}.RGB
    B=StaticSwitchParameter(True=@n2.CroppedMasked, False=@n3.RGB) {desc=CROP_EMIS, group=Emissive, parameterName=UseCroppedEmissive}
   ))
 ]}
